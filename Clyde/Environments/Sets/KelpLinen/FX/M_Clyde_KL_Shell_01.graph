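# Unlit translucent FX material for a shell effect in the Clyde KelpLinen environment, giving it a pulsing, tinted glow.
# Opacity multiplies panning noise, a downward-facing normal mask, Fresnel and a UV sine cap mask, then applies depth and location fades.
# Color blends inside and outside tints via TwoSidedTexturing, shifting toward an underlight color on downward-facing areas.
# Scrolling sine bands pulse the intensity, which is also scaled by time of day.
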Material /Clyde/Environments/Sets/KelpLinen/FX/M_Clyde_KL_Shell_01
graph {
 properties {bAutomaticallySetUsageInEditor=false, blendMode=BLEND_Translucent, shadingModel=MSM_Unlit, usedShadingModels=Unlit}
 nodes [
  n1 = TextureCoordinate()
  n2 = ComponentMask(None=@n1) {r=true}
  n3 = ScalarParameter() {defaultValue=2.5, parameterName=T_Alpha_Tiling_X}
  n4 = ComponentMask(None=@n1) {g=true}
  n5 = ScalarParameter() {defaultValue=2, parameterName=T_Alpha_Tiling_Y}
  n6 = Panner(
   Coordinate=AppendVector(A=Multiply(A=@n2, B=@n3), B=Multiply(A=@n4, B=@n5))
   Speed=AppendVector(
    A=ScalarParameter() {defaultValue=0.04, parameterName=T_Alpha_Speed_X}
    B=ScalarParameter() {defaultValue=-0.04, parameterName=T_Alpha_Speed_Y}
   )
  )
  n7 = Multiply(
   A=Power(
    Base=TextureSample(UVs=@n6) {texture=&/T_Clyde_KL_DarkAndLight_Noise}.R
    Exp=ScalarParameter() {defaultValue=1.25, parameterName=Text_Alpha_Power}
   )
   B=ScalarParameter() {defaultValue=1, parameterName=Text_Alpha_Hardness}
  )
  n8 = VertexNormalWS()
  n9 = Constant3Vector() {constant=<rgba 0 0 -1 0>}
  n10 = Multiply(A=Add(A=DotProduct(A=@n8, B=@n9)), B=ScalarParameter() {defaultValue=0.5, parameterName=Z-Up_Scale})
  n11 = ScalarParameter() {defaultValue=5, parameterName=Z-Up_Masking_Exp}
  n12 = Add(A=Clamp(None=Power(Base=@n10, Exp=@n11)), B=ScalarParameter() {defaultValue=-0.02, parameterName=Z-Up_Bias})
  n13 = Multiply(
   A=Multiply(A=Clamp(None=@n7), B=Clamp(None=@n12))
   B=MaterialFunctionCall(
    Brightness=ScalarParameter() {defaultValue=1.5, parameterName=Fresnel_Brightness}
    Exponent=ScalarParameter() {defaultValue=5, parameterName=Fresnel_Exp}
   ) {materialFunction=&MF_Fresnel_FX}.Fresnel
  )
  n14 = Sine(None=ComponentMask(None=TextureCoordinate() {vTiling=0.5}) {g=true})
  n15 = ScalarParameter() {defaultValue=1.5, parameterName=Mask_Caps_Exp}
  n16 = Clamp(None=Multiply(A=Power(Base=@n14, Exp=@n15), B=ScalarParameter() {defaultValue=1.65, parameterName=Mask_Caps_Hardness}))
  Multiply(
   A=DepthFade(Opacity=Multiply(A=@n13, B=@n16), FadeDistance=ScalarParameter() {defaultValue=250000, parameterName=FadeDistance})
   B=MaterialFunctionCall() {materialFunction=/ClydeCosmetics/Environments/KelpLinen/MF_KelpLinen_LocationFade}.Output
  )
  n17 = TextureCoordinate()
  n18 = AppendVector(
   A=ComponentMask(None=@n17) {r=true}
   B=Multiply(A=ComponentMask(None=@n17) {g=true}, B=ScalarParameter() {defaultValue=8, parameterName=Pulse_Tiling})
  )
  n19 = Time()
  n20 = AppendVector(A=Constant(), B=ScalarParameter() {defaultValue=-0.4, parameterName=Pulse_Speed})
  n21 = ComponentMask(None=Sine(None=Panner(Coordinate=@n18, Time=@n19, Speed=@n20) {speedY=1})) {g=true}
  n22 = LinearInterpolate(
   A=ScalarParameter() {defaultValue=3, parameterName=PulseA_INT}
   B=ScalarParameter() {defaultValue=6, parameterName=PulseB_INT}
   Alpha=Power(Base=ConstantBiasScale(None=@n21), Exp=ScalarParameter() {defaultValue=7, parameterName=Pulse_Exp})
  )
  n23 = VectorParameter() {parameterName=Color_Underlight, defaultValue=<rgba 150 52.5 60.6248 1>}
  n24 = Multiply(A=@n10, B=ScalarParameter() {defaultValue=3, parameterName=Z-Up_Coloring_Hardness})
  n25 = ScalarParameter() {defaultValue=1.25, parameterName=Z-Up_Coloring_Exp}
  n26 = Clamp(None=Add(A=Power(Base=@n24, Exp=@n25), B=ScalarParameter() {defaultValue=-2.4, parameterName=Bias}))
  n27 = MaterialFunctionCall!(
   Texture Side A=LinearInterpolate(A=VectorParameter() {parameterName=Color_Outside, defaultValue=<rgba 30 0.3 25.0499 1>}.RGB, B=@n23.RGB, Alpha=@n26)
   Texture Side B=LinearInterpolate(A=VectorParameter() {parameterName=Color_Inside, defaultValue=<rgba 25 0.25 16.3374 1>}.RGB, B=@n23.RGB, Alpha=@n26)
  ) {materialFunction=/Engine/Functions/Engine_MaterialFunctions03/Texturing/TwoSidedTexturing}
  n28 = Constant() {r=1}
  n29 = Constant() {r=2}
  Multiply(
   A=Multiply(
    A=Multiply(A=@n22, B=@n27.Texture Blend Output)
    B=MaterialFunctionCall(Value_Morning=@n28, Value_Day=@n29, Value_Evening=@n29, Value_Night=@n28) {materialFunction=&Scalar_Time_of_Day}.Result
   )
   B=VectorParameter() {parameterName=Tint, defaultValue=<rgba 1 1 1 1>}.RGB
  )
 ]
}
